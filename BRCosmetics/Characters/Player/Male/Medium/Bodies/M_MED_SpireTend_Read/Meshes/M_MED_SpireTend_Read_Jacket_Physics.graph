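# Physics asset for the jacket of the medium male SpireTend_Read player character body, simulating the two front jacket flap bones.
# Each flap bone has a capsule and is constrained to spine_02 with a tight swing on one axis, a wider swing on the other and no twist range.
# Larger capsules on the pelvis, lower spine and both thighs form a rough body shape, presumably as colliders for the flaps.
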PhysicsAsset /BRCosmetics/Characters/Player/Male/Medium/Bodies/M_MED_SpireTend_Read/Meshes/M_MED_SpireTend_Read_Jacket_Physics
bodies {
 dyn_jacket_fr_l [{
   shapeType=Capsule
   length=7.5
   radius=2.5
   shapeName=dyn_jacket_fr_l_capsule
   center=<9.0491 0.9874 2.2834>
   rotation=<rot 2.5494 80.435 -75.2133>
  }]
 dyn_jacket_fr_r [{shapeType=Capsule, length=9.5, radius=2.5, center=<10.4386 -1.6609 2.9322>, rotation=<rot 0 90 -75>}]
 pelvis [{shapeType=Capsule, length=12, radius=13.5, shapeName=pelvis_capsule, center=<0 -1.4639 0>}]
 spine_01 [{shapeType=Capsule, length=7, radius=12, shapeName=spine_01_capsule}]
 spine_02 [{shapeType=Capsule, length=10, radius=9, shapeName=spine_02_capsule, center=<0 0.1782 0>}]
 thigh_l [{shapeType=Capsule, length=14, radius=11.5, shapeName=thigh_l_capsule, center=<-20.2477 -0.0899 -3.1312>, rotation=<rot 0 -90 -90>}]
 thigh_r [{shapeType=Capsule, length=14, radius=11.5, shapeName=thigh_r_capsule, center=<20.2477 0.0899 3.1312>, rotation=<rot 0 -90 90>}]
}
constraints [
 {bone1Name=dyn_jacket_fr_l, bone2Name=spine_02, swing1LimitDegrees=5, swing2LimitDegrees=35, twistLimitDegrees=0}
 {bone1Name=dyn_jacket_fr_r, bone2Name=spine_02, swing1LimitDegrees=5, swing2LimitDegrees=35, twistLimitDegrees=0}
]
